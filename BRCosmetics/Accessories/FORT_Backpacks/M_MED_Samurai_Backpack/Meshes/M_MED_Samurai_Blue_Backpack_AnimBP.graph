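# Animation Blueprint for the blue Samurai backpack cosmetic, simulating the motion of its flag.
# Anim Dynamics drives the dyn_flag bone from the ref pose with an angular spring, limits and an external force.
# Each update feeds the pawn's height into Perlin noise to set that force along Y while skydiving, reduced to a tenth under parachute and zero otherwise.
# A 2D-speed noise calculation is left unconnected.
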AnimBlueprint /BRCosmetics/Accessories/FORT_Backpacks/M_MED_Samurai_Backpack/Meshes/M_MED_Samurai_Blue_Backpack_AnimBP
target_skeleton=&/M_MED_Samurai_Blue_Backpack_Skeleton
graphs [
 AnimGraph {nodes [
   n1 = Anim Dynamics - Bone: dyn_flag(
    ComponentPose=Local To Component(LocalPose=Local Space Ref Pose())
    Alpha=1
    ExternalForce=Get falling_anim_dyn_up_force_vec3() -> (falling_anim_dyn_up_force_vec3:Vector)
   ) {
    angularDampingOverride=0.98
    angularSpringConstant=250
    bAngularSpring=true
    bGravityOverrideInSimSpace=true
    bOverrideAngularDamping=true
    gravityScale=0.75
    alphaBoolBlend {blendOption=Linear}
    boundBone {boneName=dyn_flag}
    componentLinearAccScale=<1 1 1>
    componentLinearVelScale=<1 1 1>
    physicsBodyDefinitions [{
      sphereCollisionRadius=0
      boundBone {boneName=dyn_flag}
      boxExtents=<20 5 5>
      constraintSetup {angularLimitsMax=<0 0 40>, angularLimitsMin=<0 0 -40>, angularTarget=<0 0 0>}
      localJointOffset=<2 0 -1>
     }]
   }
   Output Pose(Result=Component To Local(ComponentPose=@n1)) {layerGroup=None}
  ]}
 EventGraph {nodes [
   Event BlueprintUpdateAnimation()
   n1 = Get PawnLastLocation() -> (PawnLastLocation:Vector)
   n2 = SelectFloat(
    A=Make Perlin Noise and Remap!(Value:float=BreakVector(InVec=@n1).Z, RangeOutMin:float=-25000, RangeOutMax:float=25000) -> (ReturnValue:float)
    B=0
    bPickA=Get bIsSkyDiving()
   )
   n3 = float * float(A=@n2, B=0.1)
   n4 = Get bIsParachuteOpened()
   > Set falling_anim_dyn_up_force_vec3(falling_anim_dyn_up_force_vec3:Vector=MakeVector(Y=SelectFloat(A=@n3, B=@n2, bPickA=@n4))) -> (Output_Get:Vector)
   float * float(A=float * float(A=Get PawnSpeed2D(), B=0.01), B=Make Perlin Noise and Remap(Value=0, RangeOutMin=-1000, RangeOutMax=5000))
  ]}
]
variables [falling_anim_dyn_up_force_vec3 {type=Vector}]
